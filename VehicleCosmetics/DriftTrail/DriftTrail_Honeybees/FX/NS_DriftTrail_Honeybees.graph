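# Niagara system for the Honeybees vehicle drift trail cosmetic, attached to wheel positions sampled from the vehicle skeletal mesh.
# Three emitters layer two tapering ribbons, a two-spawn-group mix of mesh and sprite particles, and two looping wheel-wrap meshes scaled per wheel.
# Each emitter takes painted color variants and masks particles per tire through FN_DisableTireFX.
# Spawn-group selectors vary particle color, lifetime and motion, while User.offset shifts the ribbons and particles from the wheels.
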NiagaraSystem /VehicleCosmetics/DriftTrail/DriftTrail_Honeybees/FX/NS_DriftTrail_Honeybees
dependencies {
 usedDataInterfaces [&NiagaraDataInterfaceCurve, &NiagaraDataInterfaceVectorCurve, &NiagaraDataInterfaceSkeletalMesh]
 usedDynamicInputs []
 usedModules [
  &self:SystemSpawnScript.SystemScriptSource.SystemScriptGraph.NiagaraNodeAssignment_0.SetVariables_#1
  &FN_UnpackVehicleFlags
  /VehicleCosmetics/FX/Niagara/Modules/VC_SetVehicleSpeedAttribs
  &SystemState
 ]
}
emitters {
 NE_VC_DriftTrail_Ribbon_Parent {
  emitterUpdateScript [
   &EmitterState()
   &VC_RateQualityHandler(Default Rate=30)
   &VC_SpawnRibbon_v1(Number of Ribbons=2, SpawnRate=$Emitter.Spawn Rate)
  ]
  particleSpawnScript [
   &VC_PaintedVariantsOverride(
    Black=<rgba 0.015 0.015 0.015 10>
    BurntSienna=<rgba 0.22 0.0524 0.0286 5>
    Cobalt=<rgba 0 0.127 1 10>
    Crimson=<rgba 0.4635 0 0.0126 5>
    DefaultColor=<rgba 0.6562 0.2809 0.0656 5>
    ForestGreen=<rgba 0 0.276 0.0139 5>
    Grey=<rgba 0.1514 0.1514 0.1615 5>
    Lime=<rgba 0.2599 0.4427 0 5>
    Orange=<rgba 0.6724 0.1426 0.0038 10>
    Pink=<rgba 0.7347 0.2348 0.7448 10>
    Purple=<rgba 0.1621 0.0606 0.8958 10>
    Saffron=<rgba 0.8854 0.6788 0 5>
    SkyBlue=<rgba 0.0197 0.3277 0.6 5>
    TitaniumWhite=<rgba 0.5596 0.5596 0.5625 5>
   )
   &VC_InitializeRibbon_v1(
    Color=$Particles.Color
    Custom Offset=true
    Enable Skeletal Position=true
    Lifetime=0.65
    Local Vector Offset=&AddVectorToPosition(Position=<-50 0 -20>, Vector=$User.offset)
    Offset Amount=1
    Ribbon Width=58
   )
   &AddVelocity(Velocity=<-1400 0 0>) {module=AddVelocity001}
  ]
  particleUpdateScript [
   &FN_DisableTireFX(Mask Index Attribute:NiagaraInt32=$Particles.SpawnGroup, Mask Index FL=2, Mask Index FR=3, Mask Index RL=0, Mask Index RR=1)
   &ParticleState()
   &InheritSourceMovement(
    Applied Position Scale=&Multiply_VectorByFloat(
     Vector=<1 1 1>
     Float=&FloatFromCurve(FloatCurve=DataInterface() {propertyValues {Curve {keys [
          {interpMode=RCIM_Cubic, leaveTangent=-0.1311, tangentMode=RCTM_Break, time=0, value=1}
          {arriveTangent=-0.0803, interpMode=RCIM_Cubic, tangentMode=RCTM_Break, time=1, value=0.9}
         ]}}})
    )
    ApplyForce=false
    ApplyPosition=true
    ApplyPositionFalloff=false
   )
   &SolveForcesAndVelocity(Clamp Velocity=false, Limit Acceleration=false)
   &ScaleColor(
    Scale Alpha=&FloatFromCurve(
     FloatCurve=DataInterface() {propertyValues {Curve {keys [
         {interpMode=RCIM_Cubic, leaveTangent=2.7108, tangentMode=RCTM_Break, time=0, value=0}
         {interpMode=RCIM_Cubic, tangentMode=RCTM_Break, time=1, value=1}
        ]}}}
     Scale Curve=3
    )
    Scale RGB=&VectorFromFloat(Value=1)
    ScaleA=true
    ScaleRGB=true
   )
   &ScaleRibbonWidth(Ribbon Width Scale=&FloatFromCurve(FloatCurve=DataInterface() {propertyValues {Curve {keys [<at 0 0.4>,<at 0.2 1>,<at 0.5688 0.6941>,<at 1 0>]}}}))
   &VC_UpdateRibbon_v1(Enable Skeletal Position=true)
  ]
  rendererClasses [&NiagaraRibbonRendererProperties]
  renderers [
   {rendererClass=&NiagaraRibbonRendererProperties}
   {rendererClass=&NiagaraRibbonRendererProperties, rendererIndex=1}
  ]
 }
 NE_VC_DriftTrail_Sprite_Parent {
  emitterUpdateScript [
   &EmitterState()
   &SpawnRate(SpawnGroup=1, SpawnRate=30, Use Spawn Probability=false) {module=SpawnRate001}
   &SpawnRate(SpawnRate=13, Use Spawn Probability=false)
  ]
  particleSpawnScript [
   &VC_PaintedVariantsOverride(
    Black=/VehicleCosmetics/FX/Niagara/Modules/MergeEmitters/VC_Dynamic_ColorBySpawnGroup(
     Color SG 0:LinearColor=<rgba 0.01 0.01 0.01 1>
     MaximumSpawnGroups:NiagaraInt32=2
     Color SG 1:LinearColor=<rgba 0.019 0.019 0.019 1>
     Color SG 2:LinearColor=<rgba 1 1 1 1>
     Color SG 3:LinearColor=<rgba 1 1 1 1>
     Color SG 4:LinearColor=<rgba 1 1 1 1>
    )
    BurntSienna=&VC_Dynamic_ColorBySpawnGroup(
     Color SG 0:LinearColor=<rgba 0.2031 0.039 0.0098 2>
     MaximumSpawnGroups:NiagaraInt32=2
     Color SG 1:LinearColor=<rgba 0.1771 0.0302 0.0132 1>
     Color SG 2:LinearColor=<rgba 1 1 1 1>
     Color SG 3:LinearColor=<rgba 1 1 1 1>
     Color SG 4:LinearColor=<rgba 1 1 1 1>
    )
    Cobalt=&VC_Dynamic_ColorBySpawnGroup(
     Color SG 0:LinearColor=<rgba 0.0026 0.1243 0.75 2>
     MaximumSpawnGroups:NiagaraInt32=2
     Color SG 1:LinearColor=<rgba 0 0.0673 0.375 2>
     Color SG 2:LinearColor=<rgba 1 1 1 1>
     Color SG 3:LinearColor=<rgba 1 1 1 1>
     Color SG 4:LinearColor=<rgba 1 1 1 1>
    )
    Crimson=&VC_Dynamic_ColorBySpawnGroup(
     Color SG 0:LinearColor=<rgba 0.401 0 0.0109 1>
     MaximumSpawnGroups:NiagaraInt32=2
     Color SG 1:LinearColor=<rgba 0.4635 0 0.0126 1>
     Color SG 2:LinearColor=<rgba 1 1 1 1>
     Color SG 3:LinearColor=<rgba 1 1 1 1>
     Color SG 4:LinearColor=<rgba 1 1 1 1>
    )
    DefaultColor=&VC_Dynamic_ColorBySpawnGroup(
     Color SG 0:LinearColor=<rgba 0.5052 0.2295 0.0105 1>
     MaximumSpawnGroups:NiagaraInt32=2
     Color SG 1:LinearColor=<rgba 0.5625 0.1833 0.0352 7>
     Color SG 2:LinearColor=<rgba 1 1 1 1>
     Color SG 3:LinearColor=<rgba 1 1 1 1>
     Color SG 4:LinearColor=<rgba 1 1 1 1>
    )
    ForestGreen=&VC_Dynamic_ColorBySpawnGroup(
     Color SG 0:LinearColor=<rgba 0 0.25 0.0126 1>
     MaximumSpawnGroups:NiagaraInt32=2
     Color SG 1:LinearColor=<rgba 0 0.4062 0.0205 1>
     Color SG 2:LinearColor=<rgba 1 1 1 1>
     Color SG 3:LinearColor=<rgba 1 1 1 1>
     Color SG 4:LinearColor=<rgba 1 1 1 1>
    )
    Grey=&VC_Dynamic_ColorBySpawnGroup(
     Color SG 0:LinearColor=<rgba 0.1322 0.1322 0.151 1>
     MaximumSpawnGroups:NiagaraInt32=2
     Color SG 1:LinearColor=<rgba 0.1823 0.1823 0.2083 1>
     Color SG 2:LinearColor=<rgba 1 1 1 1>
     Color SG 3:LinearColor=<rgba 1 1 1 1>
     Color SG 4:LinearColor=<rgba 1 1 1 1>
    )
    Lime=&VC_Dynamic_ColorBySpawnGroup(
     Color SG 0:LinearColor=<rgba 0.2404 0.4167 0 1>
     MaximumSpawnGroups:NiagaraInt32=2
     Color SG 1:LinearColor=<rgba 0.349 0.5885 0 1>
     Color SG 2:LinearColor=<rgba 1 1 1 1>
     Color SG 3:LinearColor=<rgba 1 1 1 1>
     Color SG 4:LinearColor=<rgba 1 1 1 1>
    )
    Orange=&VC_Dynamic_ColorBySpawnGroup(
     Color SG 0:LinearColor=<rgba 0.5729 0.1171 0 1>
     MaximumSpawnGroups:NiagaraInt32=2
     Color SG 1:LinearColor=<rgba 0.6724 0.1274 0.0005 1>
     Color SG 2:LinearColor=<rgba 1 1 1 1>
     Color SG 3:LinearColor=<rgba 1 1 1 1>
     Color SG 4:LinearColor=<rgba 1 1 1 1>
    )
    Pink=&VC_Dynamic_ColorBySpawnGroup(
     Color SG 0:LinearColor=<rgba 0.6042 0.2163 0.5563 1>
     MaximumSpawnGroups:NiagaraInt32=2
     Color SG 1:LinearColor=<rgba 0.5156 0.1423 0.4696 1>
     Color SG 2:LinearColor=<rgba 1 1 1 1>
     Color SG 3:LinearColor=<rgba 1 1 1 1>
     Color SG 4:LinearColor=<rgba 1 1 1 1>
    )
    Purple=&VC_Dynamic_ColorBySpawnGroup(
     Color SG 0:LinearColor=<rgba 0.1528 0.0817 0.6719 1>
     MaximumSpawnGroups:NiagaraInt32=2
     Color SG 1:LinearColor=<rgba 0.1213 0.0521 0.7552 1>
     Color SG 2:LinearColor=<rgba 1 1 1 1>
     Color SG 3:LinearColor=<rgba 1 1 1 1>
     Color SG 4:LinearColor=<rgba 1 1 1 1>
    )
    SkyBlue=&VC_Dynamic_ColorBySpawnGroup(
     Color SG 0:LinearColor=<rgba 0.0451 0.3343 0.5729 1>
     MaximumSpawnGroups:NiagaraInt32=2
     Color SG 1:LinearColor=<rgba 0.1011 0.4087 0.6458 1>
     Color SG 2:LinearColor=<rgba 1 1 1 1>
     Color SG 3:LinearColor=<rgba 1 1 1 1>
     Color SG 4:LinearColor=<rgba 1 1 1 1>
    )
    TitaniumWhite=<rgba 0.4905 0.491 0.5208 1>
   )
   /SetVariables(Particles.TagHoney:NiagaraInt32=&SpawnGroup()) {isSetParameters=true}
   &InitializeParticle(
    Color=$Particles.Color
    Lifetime=/VehicleCosmetics/FX/Niagara/Modules/MergeEmitters/VC_FloatBySpawnGroup(
     SpawnGroup or ExecIndex:NiagaraInt32=0
     CustomInt:NiagaraInt32=0
     MaximumSpawnGroups:NiagaraInt32=2
     Float SG 0:NiagaraFloat=&RandomRangeFloat(Minimum=0.4, Maximum=0.4)
     Float SG 1:NiagaraFloat=&RandomRangeFloat(Minimum=0.3, Maximum=0.55)
     Float SG 2:NiagaraFloat=0
     Float SG 3:NiagaraFloat=0
     Float SG 4:NiagaraFloat=0
     Float SG 5:NiagaraFloat=0
    )
    Mesh Scale=&Multiply_VectorByFloat(Vector=<0.6 0.7 0.6>, Float=3)
    Mesh Scale Mode=::NewEnumerator4
    Sprite Rotation Mode=::NewEnumerator2
    Sprite Size=&RandomRangeVector2D(
     Minimum=&Multiply_Vector2D_ByFloat(Float=1.5, Vector2D=<5 5>)
     Maximum=&Multiply_Vector2D_ByFloat(Float=1.5, Vector2D=<7.8 7.8>)
    )
    Sprite Size Mode=::NewEnumerator4
    UsePositionOffset=false
    Write Material Random=true
   )
   &PartitionParticles(Particles Per Partition=4, Partition Mode:ENiagara_PartitionMode=::NewEnumerator1)
   &SkeletalMeshLocation(
    Bone / Socket Index=&Add_Integer(A=$Particles.PartitionParticles.Partition, B=4)
    Bone / Socket Sampling Mode=::NewEnumerator5
    Interpolate Mesh Sampling=false
    MaskSpawnGroup=false
    Mesh Sampling Type=::NewEnumerator6
    Sampled Position Offset=&AddVectorToPosition(
     Position=&MaskVectorBySpawnGroup(Masked Vector=<-40 0 -20>, Passthrough Vector=<-50 0 -20>)
     Vector=$User.offset
    )
    Skeletal Mesh=$System.SkeletalMesh
   )
   &InitialMeshOrientation(
    Mesh Axis To Orient=<0 0 0>
    Mesh Orientation Mode=::NewEnumerator2
    Orientation Vector=<0 0 1>
    Rotation=&RandomRangeVector(Minimum=<0 0 -1>, Maximum=<0 0 1>)
    Use Rotation Vector=true
   )
   &AddVelocity(Velocity=/VehicleCosmetics/FX/Niagara/Modules/MergeEmitters/VC_VectorBySpawnGroup(
     SpawnGroup or ExecIndex:NiagaraInt32=0
     CustomInt:NiagaraInt32=0
     MaximumSpawnGroups:NiagaraInt32=2
     Vector Group 0:Vector3f=<0 0 0>
     Vector Group 1:Vector3f=&RandomRangeVector(Minimum=<-60 12 20>, Maximum=<-20 -12 0>)
     Vector Group 2:Vector3f=<0 0 0>
     Vector Group 3:Vector3f=<0 0 0>
     Vector Group 4:Vector3f=<0 0 0>
     Vector Group 5:Vector3f=<0 0 0>
    ))
  ]
  particleUpdateScript [
   &FN_DisableTireFX(
    Mask Index Attribute:NiagaraInt32=$Particles.PartitionParticles.Partition
    Mask Index FL=2
    Mask Index FR=3
    Mask Index RL=0
    Mask Index RR=1
   )
   &ParticleState()
   &UpdateMeshOrientation(Rotation Vector=&RandomRangeVector(Minimum=<0 0 0.3>, Maximum=<0 0 0.5>))
   &ScaleVelocity(Velocity Scale=&VC_VectorBySpawnGroup(
     SpawnGroup or ExecIndex:NiagaraInt32=0
     CustomInt:NiagaraInt32=0
     MaximumSpawnGroups:NiagaraInt32=2
     Vector Group 0:Vector3f=<0 0 0>
     Vector Group 1:Vector3f=&VectorFromCurve(
      VectorCurve=DataInterface() {propertyValues {
        XCurve {
         defaultValue=1
         keys [
          {interpMode=RCIM_Cubic, tangentMode=RCTM_Break, time=0, value=1}
          {arriveTangent=-2.8919, arriveTangentWeight=0.3333, interpMode=RCIM_Cubic, tangentMode=RCTM_Break, time=1, value=0}
         ]
        }
        YCurve {
         defaultValue=1
         keys [
          {interpMode=RCIM_Cubic, tangentMode=RCTM_Break, time=0, value=1}
          {arriveTangent=-2.8919, arriveTangentWeight=0.3333, interpMode=RCIM_Cubic, tangentMode=RCTM_Break, time=1, value=0}
         ]
        }
        ZCurve {keys [
          {interpMode=RCIM_Cubic, tangentMode=RCTM_Break, time=0, value=1}
          {arriveTangent=-2.8919, interpMode=RCIM_Cubic, tangentMode=RCTM_Break, time=1, value=0}
         ]}
       }}
      Scale Curve=<5 5 5>
     )
     Vector Group 2:Vector3f=<0 0 0>
     Vector Group 3:Vector3f=<0 0 0>
     Vector Group 4:Vector3f=<0 0 0>
     Vector Group 5:Vector3f=<0 0 0>
    ))
   &ScaleSpriteSize(
    Uniform Curve Scale=2.5
    Uniform Curve Sprite Scale=DataInterface() {propertyValues {Curve {keys [
        {interpMode=RCIM_Cubic, tangentMode=RCTM_Break, time=0.2, value=1}
        {arriveTangent=-2.1991, interpMode=RCIM_Cubic, tangentMode=RCTM_Break, time=1, value=0.2}
       ]}}}
   )
   &ScaleMeshSize(Scale Factor=&VectorFromCurve(VectorCurve=DataInterface() {propertyValues {
       XCurve {defaultValue=0.1, keys [<at 0 0.35>,<at 0.45 0.8>,<at 1 1>]}
       YCurve {keys [<at 0 0.35>,<at 0.45 0.8>,<at 1 1>]}
       ZCurve {defaultValue=0.35, keys [<at 0 0.3 RCIM_Cubic>, {arriveTangent=-0.4627, interpMode=RCIM_Cubic, tangentMode=RCTM_Break, time=1, value=4}]}
      }})) {module=ScaleMeshSize001}
   &ScaleColor(
    Scale Alpha=&VC_FloatBySpawnGroup(
     SpawnGroup or ExecIndex:NiagaraInt32=0
     CustomInt:NiagaraInt32=0
     MaximumSpawnGroups:NiagaraInt32=2
     Float SG 0:NiagaraFloat=&FloatFromCurve(FloatCurve=DataInterface() {propertyValues {Curve {keys [<at 0.7 1>,<at 1 0>]}}})
     Float SG 1:NiagaraFloat=1
     Float SG 2:NiagaraFloat=0
     Float SG 3:NiagaraFloat=0
     Float SG 4:NiagaraFloat=0
     Float SG 5:NiagaraFloat=0
    )
    Scale RGB=&VC_VectorBySpawnGroup(
     SpawnGroup or ExecIndex:NiagaraInt32=0
     CustomInt:NiagaraInt32=0
     MaximumSpawnGroups:NiagaraInt32=2
     Vector Group 0:Vector3f=<2.3 2.3 2.3>
     Vector Group 1:Vector3f=&VectorFromFloat(Value=0.55)
     Vector Group 2:Vector3f=<0 0 0>
     Vector Group 3:Vector3f=<0 0 0>
     Vector Group 4:Vector3f=<0 0 0>
     Vector Group 5:Vector3f=<0 0 0>
    )
    ScaleA=true
    ScaleRGB=true
   )
   &DynamicMaterialParameters(
    Index 0 Param 1=&FloatFromCurve(FloatCurve=DataInterface() {propertyValues {Curve {keys [<at 0 1>,<at 1 1.5>]}}})
    Index 0 Param 2=&FloatFromCurve(FloatCurve=DataInterface() {propertyValues {Curve {keys [<at 0.45 -1>,<at 1 0>]}}})
    Index 0 Param 3=0.192
    Index 1 Param 1=1
    Index 1 Param 3=1
    Index 1 Param 4=2
    Module.Index0Float0Write=true
    Module.Index0Float1Write=true
    Module.Index0Float2Write=true
    Module.Index0Float3Write=true
    Module.Index1Float0Write=true
    Module.Index1Float1Write=true
    Module.Index1Float2Write=true
    Module.Index1Float3Write=true
    Write Parameter Index 1=::NewEnumerator1
   )
   &ScaleVelocity(Velocity Scale=&MaskVectorBySpawnGroup(
     Masked Vector=<1 1 1>
     Passthrough Vector=&VectorFromCurve(VectorCurve=DataInterface() {propertyValues {XCurve {keys [<at 0 4>]}, YCurve {keys [<at 0 4>]}, ZCurve {keys [<at 0 4>,<at 0.5 4>,<at 1 -10>]}}})
    )) {module=ScaleVelocity001}
   &SolveForcesAndVelocity(Clamp Velocity=false, Limit Acceleration=false)
  ]
  rendererClasses [&NiagaraMeshRendererProperties, &NiagaraSpriteRendererProperties]
  renderers [
   {rendererClass=&NiagaraMeshRendererProperties, rendererIndex=0}
   {rendererClass=&NiagaraSpriteRendererProperties, rendererIndex=1}
  ]
 }
 NE_VC_DriftTrail_WheelWrap_Parent {
  emitterUpdateScript [
   &EmitterState(
    Inactive Response=::NewEnumerator1
    Life Cycle Mode=::NewEnumerator1
    Loop Behavior=::NewEnumerator1
    Loop Duration Mode=::NewEnumerator1
    UseLoopDelay=false
   )
   &SpawnBurst_Instantaneous(Spawn Count=2, Use Spawn Probability=false, UseLoopCountLimit=false)
  ]
  particleSpawnScript [
   &VC_PaintedVariantsOverride(
    Black=<rgba 0.015 0.015 0.015 10>
    BurntSienna=<rgba 0.2396 0.057 0.0311 1>
    Cobalt=<rgba 0 0.1524 1.2 1>
    Crimson=<rgba 0.4635 0 0.0126 5>
    DefaultColor=<rgba 0.8854 0.355 0.0507 1>
    ForestGreen=<rgba 0 0.276 0.0139 5>
    Grey=<rgba 0.1514 0.1514 0.1615 5>
    Lime=<rgba 0.2599 0.4427 0 5>
    Orange=<rgba 0.6724 0.1426 0.0038 10>
    Pink=<rgba 0.7347 0.2348 0.7448 10>
    Purple=<rgba 0.1621 0.0606 0.8958 10>
    Saffron=<rgba 0.8854 0.6788 0 5>
    SkyBlue=<rgba 0.0197 0.3277 0.6 5>
    TitaniumWhite=<rgba 0.5596 0.5596 0.5625 5>
   )
   &InitializeParticle(
    Color Mode=::NewEnumerator0
    Mesh Scale Mode=::NewEnumerator4
    Sprite Rotation Mode=::NewEnumerator2
    Sprite Size Mode=::NewEnumerator3
    UsePositionOffset=false
    Write Material Random=true
   )
   &PartitionParticles()
   &SkeletalMeshLocation(
    Bone / Socket Index=&Add_Integer(A=$Particles.PartitionParticles.Partition, B=4)
    Bone / Socket Sampling Mode=::NewEnumerator5
    Interpolate Mesh Sampling=false
    MaskSpawnGroup=false
    Mesh Sampling Type=::NewEnumerator6
    Skeletal Mesh=$System.SkeletalMesh
   )
  ]
  particleUpdateScript [
   &FN_DisableTireFX(
    Mask Index Attribute:NiagaraInt32=$Particles.PartitionParticles.Partition
    Mask Index FL=2
    Mask Index FR=3
    Mask Index RL=0
    Mask Index RR=1
   )
   &VC_WheelWrapScale(Wheel Index Attribute:NiagaraInt32=&Add_Integer(A=$Particles.PartitionParticles.Partition, B=2))
   &ParticleState(Kill Particles When Lifetime Has Elapsed=false, Loop Particles Lifetime=true)
   &ScaleColor(Scale RGB=&VectorFromFloat(Value=3), ScaleA=true, ScaleRGB=true)
   &UpdateMeshOrientation(Rotation Coordinate Space=::NewEnumerator2, Rotation Vector=<0 1 0>) {enabled=false}
   &SkeletalMeshLocation(
    Bone / Socket Index=&Add_Integer(A=$Particles.PartitionParticles.Partition, B=4)
    Bone / Socket Sampling Mode=::NewEnumerator5
    MaskSpawnGroup=false
    Mesh Sampling Type=::NewEnumerator6
    Skeletal Mesh=$System.SkeletalMesh
   ) {module=SkeletalMeshLocation001}
   &SolveForcesAndVelocity(Clamp Velocity=false, Limit Acceleration=false)
  ]
  rendererClasses [&NiagaraMeshRendererProperties]
  renderers [{rendererClass=&NiagaraMeshRendererProperties, rendererIndex=0}]
 }
}
summary {
 systemName=NS_DriftTrail_Honeybees
 emitters [
  {emitterName=NE_VC_DriftTrail_Ribbon_Parent, rendererClasses [&NiagaraRibbonRendererProperties]}
  {
   emitterName=NE_VC_DriftTrail_Sprite_Parent
   rendererClasses [&NiagaraMeshRendererProperties, &NiagaraSpriteRendererProperties]
  }
  {emitterName=NE_VC_DriftTrail_WheelWrap_Parent, rendererClasses [&NiagaraMeshRendererProperties]}
 ]
 userVariables [
  User.bIsFrontendPreview {type=NiagaraBool}
  User.bIsPainted {type=NiagaraBool}
  User.ColorVariantID {type=NiagaraInt32}
  User.offset {type=Vector3f, description="offset from wheels(set through cosmetic template)", defaultValue {struct=Vector3f, value=<0 0 0>}}
  User.PaintedColor {type=LinearColor, description="", defaultValue {struct=LinearColor, value=<rgba 1 1 1 1>}}
  User.PaintedScalar {type=NiagaraFloat, defaultValue=1, description=""}
  User.Shine {type=NiagaraFloat, defaultValue=5, description=""}
  User.SpecMult {type=NiagaraFloat, defaultValue=0, description=""}
  User.SpecPower {type=NiagaraFloat, defaultValue=0, description=""}
  User.VehicleFlags {type=NiagaraInt32, defaultValue=806289408}
  User.Wheel_Radius_Array {
   type=NiagaraDataInterfaceArrayFloat
   defaultValue {
    struct=NiagaraExt_VariableValue_DataInterface
    value {
     dataInterface=&self:NiagaraDataInterfaceArrayFloat_1
     dataInterfaceClass=&NiagaraDataInterfaceArrayFloat
    }
   }
  }
  User.Wheel_Width_Array {
   type=NiagaraDataInterfaceArrayFloat
   defaultValue {
    struct=NiagaraExt_VariableValue_DataInterface
    value {
     dataInterface=&self:NiagaraDataInterfaceArrayFloat_0
     dataInterfaceClass=&NiagaraDataInterfaceArrayFloat
    }
   }
  }
  User.WheelTouchingArray {
   type=NiagaraDataInterfaceArrayFloat
   defaultValue {
    struct=NiagaraExt_VariableValue_DataInterface
    value {
     dataInterface=&self:NiagaraDataInterfaceArrayFloat_2
     dataInterfaceClass=&NiagaraDataInterfaceArrayFloat
    }
   }
  }
 ]
}
